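# Post-process AnimBP for the SpendPlume_Shove biped companion cosmetic, layering deformation fixes and secondary motion onto the input pose.
# Constraints drive thigh deform bones and a weighted belt bone, then spring-based AnimDynamics simulates four belt bones.
# Small foot offsets switch on IsPBRVariant, set at init when the outfit lacks the Stage1 variant tag.
# Finally, curves remapped from the attached parent mesh through SpendPlume's blink expressions feed its pose asset.
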AnimBlueprint /CosmeticCompanions/Assets/Biped/SpendPlume_Shove/Meshes/SpendPlume_Shove_Biped_Post_AnimBP
target_skeleton=&/SpendPlume_Shove_Biped_Skeleton
graphs [
 AnimGraph {nodes [
   n1 = "Constraint - thigh_deform_l (TT)"!(ComponentPose=Local To Component(LocalPose=Input Pose()), Alpha=1, ConstraintWeights_0=1, ConstraintWeights_1:float=1) -> (Pose) {
    boneToModify {boneName=thigh_deform_l}
    constraintSetup [{transformType=Translation, targetBone {boneName=thigh_l}}, {transformType=Translation, targetBone {boneName=foot_l}}]
    constraintWeights [1,1]
   }
   n2 = "Constraint - thigh_deform_r (TT)"!(ComponentPose=@n1, Alpha=1, ConstraintWeights_0=1, ConstraintWeights_1:float=1) -> (Pose) {
    boneToModify {boneName=thigh_deform_r}
    constraintSetup [{transformType=Translation, targetBone {boneName=thigh_r}}, {transformType=Translation, targetBone {boneName=foot_r}}]
    constraintWeights [1,1]
   }
   n3 = "Constraint - deform_belt (PPPPPPPP)"!(
    ComponentPose=@n2
    Alpha=1
    ConstraintWeights_0=0.4518
    ConstraintWeights_1:float=0.2034
    ConstraintWeights_2:float=0.1147
    ConstraintWeights_3:float=0.1101
    ConstraintWeights_4:float=0.0988
    ConstraintWeights_5:float=0.0127
    ConstraintWeights_6:float=0.0047
    ConstraintWeights_7:float=0.0037
   ) -> (Pose) {
    boneToModify {boneName=deform_belt}
    constraintSetup [
     {targetBone {boneName=spine_01}}
     {targetBone {boneName=spine_02}}
     {targetBone {boneName=thigh_deform_l}}
     {targetBone {boneName=thigh_deform_r}}
     {targetBone {boneName=spine_03}}
     {targetBone {boneName=spine_04}}
     {targetBone {boneName=pelvis}}
     {targetBone {boneName=neck_01}}
    ]
    constraintWeights [1,1,1,1,1,1,1,1]
   }
   n4 = Anim Dynamics - Bone: deform_belt_01_l(ComponentPose=@n3, Alpha=1, ExternalForce="0,0,0") {
    angularDampingOverride=1
    angularSpringConstant=20000
    bAngularSpring=true
    bOverrideAngularDamping=true
    bUseGravityOverride=true
    gravityScale=0.35
    simulationSpace=RootRelative
    boundBone {boneName=deform_belt_01_l}
    componentLinearAccScale=<1 1 1>
    componentLinearVelScale=<1 1 1>
    gravityOverride {z=-980}
    physicsBodyDefinitions [{
      boundBone {boneName=deform_belt_01_l}
      boxExtents=<20 30 30>
      constraintSetup {angularLimitsMax=<15 5 15>, angularLimitsMin=<-15 -15 -15>, angularTarget=<1 0 0>}
      localJointOffset=<4 0 0>
     }]
   }
   n5 = Anim Dynamics - Bone: deform_belt_02_l(ComponentPose=@n4, Alpha=1, ExternalForce="0,0,0") {
    angularDampingOverride=1
    angularSpringConstant=20000
    bAngularSpring=true
    bOverrideAngularDamping=true
    bUseGravityOverride=true
    gravityScale=0.35
    simulationSpace=RootRelative
    boundBone {boneName=deform_belt_02_l}
    componentLinearAccScale=<1 1 1>
    componentLinearVelScale=<1 1 1>
    gravityOverride {z=-980}
    physicsBodyDefinitions [{
      boundBone {boneName=deform_belt_02_l}
      boxExtents=<20 30 30>
      constraintSetup {angularLimitsMax=<15 5 15>, angularLimitsMin=<-15 -15 -15>, angularTarget=<1 0 0>}
      localJointOffset=<4 0 0>
     }]
   }
   n6 = Anim Dynamics - Bone: deform_belt_01_r(ComponentPose=@n5, Alpha=1, ExternalForce="0,0,0") {
    angularDampingOverride=1
    angularSpringConstant=20000
    bAngularSpring=true
    bOverrideAngularDamping=true
    bUseGravityOverride=true
    gravityScale=0.35
    simulationSpace=RootRelative
    boundBone {boneName=deform_belt_01_r}
    componentLinearAccScale=<1 1 1>
    componentLinearVelScale=<1 1 1>
    gravityOverride {z=-980}
    physicsBodyDefinitions [{
      boundBone {boneName=deform_belt_01_r}
      boxExtents=<20 30 30>
      constraintSetup {angularLimitsMax=<15 5 15>, angularLimitsMin=<-15 -15 -15>, angularTarget=<1 0 0>}
      localJointOffset=<4 0 0>
     }]
   }
   n7 = Anim Dynamics - Bone: deform_belt_02_r(ComponentPose=@n6, Alpha=1, ExternalForce="0,0,0") {
    angularDampingOverride=1
    angularSpringConstant=20000
    bAngularSpring=true
    bOverrideAngularDamping=true
    bUseGravityOverride=true
    gravityScale=0.35
    simulationSpace=RootRelative
    boundBone {boneName=deform_belt_02_r}
    componentLinearAccScale=<1 1 1>
    componentLinearVelScale=<1 1 1>
    gravityOverride {z=-980}
    physicsBodyDefinitions [{
      boundBone {boneName=deform_belt_02_r}
      boxExtents=<20 30 30>
      constraintSetup {angularLimitsMax=<15 5 15>, angularLimitsMin=<-15 -15 -15>, angularTarget=<1 0 0>}
      localJointOffset=<4 0 0>
     }]
   }
   n8 = Get IsPBRVariant() -> (IsPBRVariant:bool)
   n9 = "Transform (Modify) Bone - Bone: foot_l"!(ComponentPose=@n7, bAlphaBoolEnabled:bool=@n8, Translation="-0.35,0,0", Rotation="0,0,0", Scale="1,1,1") -> (Pose) {alphaInputType=Bool, translationMode=BMM_Additive, translationSpace=BCS_BoneSpace, boneToModify {boneName=foot_l}, translation {x=-0.35}}
   n10 = "Transform (Modify) Bone - Bone: foot_r"!(ComponentPose=@n9, bAlphaBoolEnabled:bool=@n8, Translation="0.35,0,0", Rotation="0,0,0", Scale="1,1,1") -> (Pose) {alphaInputType=Bool, translationMode=BMM_Additive, translationSpace=BCS_BoneSpace, boneToModify {boneName=foot_r}, translation {x=0.35}}
   n11 = NOT Boolean(A=Get IsPBRVariant())
   n12 = "Transform (Modify) Bone - Bone: foot_l"(ComponentPose=@n10, bAlphaBoolEnabled=@n11, Translation="0.15,0,0", Rotation="0,0,0", Scale="1,1,1") {alphaInputType=Bool, translationMode=BMM_Additive, translationSpace=BCS_BoneSpace, boneToModify {boneName=foot_l}, translation {x=0.15}}
   n13 = "Transform (Modify) Bone - Bone: foot_r"(ComponentPose=@n12, bAlphaBoolEnabled=@n11, Translation="-0.15,0,0", Rotation="0,0,0", Scale="1,1,1") {alphaInputType=Bool, translationMode=BMM_Additive, translationSpace=BCS_BoneSpace, boneToModify {boneName=foot_r}, translation {x=-0.15}}
   n14 = Component To Local(ComponentPose=@n13)
   n15 = Get mesh component() -> (mesh component:Skeletal Mesh Component Object Reference)
   Output Pose(Result=Pose!(
     PoseAsset:Pose Asset Object Reference=Get PoseAsset() -> (PoseAsset:Pose Asset Object Reference)
     SourcePose:Pose Link Structure=Remap Curves From Mesh!(
      SourcePose:Pose Link Structure=@n14
      CurveExpressionsDataAsset:Curve Expressions Data Asset Object Reference=/CosmeticCompanions/Assets/Biped/SpendPlume/Meshes/SpendPlume_Blink_Remap
      SourceMeshComponent:Skeletal Mesh Component Weak object=@n15
     ) -> (Pose:Pose Link Structure) {is=title AnimGraphNode_RemapCurvesFromMesh, bUseAttachedParent=true, expressionSource=DataAsset}
    ) -> (Pose:Pose Link Structure) {is=title AnimGraphNode_PoseBlendNode, blendOption=Linear})
  ]}
 EventGraph {nodes [
   Event BlueprintInitializeAnimation()
   > Set mesh component(mesh component:Skeletal Mesh Component Object Reference=GetOwningComponent()) -> (Output_Get:Skeletal Mesh Component Object Reference)
   n1 = GetMainAnimInstance!(self:Companion Post Process Anim Instance Object Reference) -> (ReturnValue:Fort Player Companion Anim Instance Object Reference)
   > Set IsPBRVariant(IsPBRVariant:bool=NOT Boolean(A=HasVariantTagForChannel!(
      self:Fort Player Companion Anim Instance Object Reference=@n1
      ChannelTag:Gameplay Tag Structure=(TagName="Cosmetics.Variant.Channel.Outfit")
      VariantTag:Gameplay Tag Structure=(TagName="Cosmetics.Variant.Property.Stage1")
     ) -> (ReturnValue:bool))) -> (Output_Get:bool)
  ]}
]
variables [
 IsPBRVariant {type=boolean}
 mesh component {type=SkeletalMeshComponent}
 PoseAsset {type=PoseAsset, default=/CosmeticCompanions/Assets/Biped/SpendPlume/Meshes/SpendPlume_Biped_PoseAsset}
]
